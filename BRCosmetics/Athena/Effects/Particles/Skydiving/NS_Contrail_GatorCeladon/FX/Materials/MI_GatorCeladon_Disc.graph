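# Translucent, unlit, two-sided Niagara mesh-particle material for the disc in the GatorCeladon skydiving contrail effect.
# Skewed, noise-distorted UVs pan a smoke texture, while a panning Voronoi texture supplies a dissolve mask.
# These patterns shape a two-color emissive gradient with a third color at one V edge, plus opacity with V falloff and particle alpha.
# Niagara dynamic parameters drive pan speed and a smoothstep reveal wipe along V.
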Material /BRCosmetics/Athena/Effects/Particles/Skydiving/NS_Contrail_GatorCeladon/FX/Materials/MI_GatorCeladon_Disc
graph {
 properties {
  bAutomaticallySetUsageInEditor=false
  bUsedWithNiagaraMeshParticles=true
  blendMode=BLEND_Translucent
  pixelDepthOffsetMode=PDOM_Legacy
  shadingModel=MSM_Unlit
  twoSided=true
  usedShadingModels=Unlit
 }
 nodes [
  n1 = VectorParameter() {parameterName=ColorA, defaultValue=<rgba 0.02 0.01 0.006 1>}
  n2 = VectorParameter() {parameterName=ColorB, defaultValue=<rgba 0.2 0.0167 0 1>}
  n3 = NamedRerouteUsage()
  n4 = NamedRerouteUsage()
  n5 = VectorParameter() {parameterName=Param_1, defaultValue=<rgba 5 1 0 1>}
  n6 = MaterialFunctionCall(X=@n5.R, Y=@n5.R) {materialFunction=&MakeFloat2}
  n7 = TextureSample(UVs=Panner(
    Coordinate=Multiply(A=@n4, B=@n6.Result)
    Time=DynamicParameter() {paramNames [Speed, Reveal, WPO Width, WPO MoweAway]}.Speed
    Speed=MaterialFunctionCall(X=@n5.B, Y=@n5.B) {materialFunction=&MakeFloat2}.Result
   )) {texture=/BRCosmetics/Athena/Effects/Particles/Skydiving/NS_Contrail_PinkSpike/Textures/T_PinkSpike_Voronoi2D_Height}
  n8 = TextureCoordinate()
  n9 = SmoothStep(
   Min=@n7.R
   Max=Add(A=@n7.R, B=ScalarParameter() {defaultValue=0.6, parameterName=Smooth})
   Value=Power(
    Base=MaterialFunctionCall(Float2=@n8) {materialFunction=&BreakOutFloat2Components}.G
    Exp=Constant() {r=1}
   )
  )
  n10 = Saturate(None=Multiply(A=Multiply(A=@n3, B=@n9), B=Constant() {r=2}))
  n11 = Power(
   Base=MaterialFunctionCall(Float2=TextureCoordinate()) {materialFunction=&BreakOutFloat2Components}.G
   Exp=Constant() {r=15}
  )
  MaterialFunctionCall(Emissive=LinearInterpolate(
    A=VectorParameter() {parameterName=ColorC, defaultValue=<rgba 1 0.0333 0 1>}.RGB
    B=LinearInterpolate(A=@n1.RGB, B=@n2.RGB, Alpha=@n10)
    Alpha=OneMinus(None=@n11)
   )) {materialFunction=&MF_FX_EmissiveAdjustments}
  n12 = MaterialFunctionCall(Float2=TextureCoordinate()) {materialFunction=&BreakOutFloat2Components}
  n13 = Add(A=@n12.R, B=Multiply(A=@n12.G, B=Constant() {r=0.2}))
  n14 = Add(
   A=NamedRerouteDeclaration(None=MaterialFunctionCall(X=@n13, Y=@n12.G) {materialFunction=&MakeFloat2}.Result) {name=SkewedUVs, nodeColor=<rgba 0 1 0.0625 1>}
   B=Multiply(A=NamedRerouteUsage(), B=ScalarParameter() {defaultValue=0.1, group=Distortion, parameterName=DistortionAmount, sortPriority=3}) {constB=0.2}
  )
  n15 = VectorParameter() {
   group=MainTexture
   parameterName=FlameTextureUVControl
   sortPriority=2
   channelNames {a=SpeedY, b=SpeedX, g=ScaleY, r=ScaleX}
   defaultValue=<rgba 4 1 0 0.6>
  }
  n16 = ComponentMask(None=@n15) {g=true, r=true}
  n17 = TextureSampleParameter2D(UVs=Panner(
    Coordinate=Multiply(A=@n14, B=@n16)
    Time=DynamicParameter() {paramNames [Speed, Reveal, WPO Width, WPO MoweAway]}.Speed
    Speed=AppendVector(A=@n15.SpeedX, B=@n15.SpeedX)
   ) {speedX=0.6}) {group=MainTexture, parameterName=MainTexture, sortPriority=1, texture=&T_Inky_Smoke_Tile}
  NamedRerouteDeclaration(None=@n17.R) {name=Base, nodeColor=<rgba 0.6875 0 1 1>}
  n18 = Constant() {r=0.1}
  n19 = Constant() {r=0.9}
  n20 = MaterialFunctionCall(Float2=TextureCoordinate()) {materialFunction=&BreakOutFloat2Components}
  n21 = OneMinus(None=Power(Base=OneMinus(None=@n20.G), Exp=Constant() {r=2}))
  n22 = Constant()
  n23 = SmoothStep(Min=@n22, Max=Add(A=@n22, B=Constant() {r=0.1}), Value=OneMinus(None=@n20.G))
  n24 = Saturate(None=Multiply(A=LinearInterpolate(A=@n18, B=@n19, Alpha=@n21), B=Saturate(None=@n23)))
  n25 = NamedRerouteUsage()
  NamedRerouteDeclaration(None=Multiply(A=Multiply(A=@n24, B=@n25), B=ParticleColor().A)) {name=Opacity, nodeColor=<rgba 0 1 0.1094 1>}
  n26 = VectorParameter() {
   group=Distortion
   parameterName=DistortionUVControl
   sortPriority=2
   channelNames {a=SpeedY, b=SpeedX, g=ScaleY, r=ScaleX}
   defaultValue=<rgba 1 0.5 0.1 0.1>
  }
  n27 = Multiply(A=TextureCoordinate(), B=ComponentMask(None=@n26) {g=true, r=true}) {constB=0.6}
  n28 = AppendVector(A=@n26.SpeedX, B=@n26.SpeedX)
  NamedRerouteDeclaration(None=TextureSample(UVs=Panner(Coordinate=@n27, Speed=@n28) {speedX=1, speedY=0.25}) {texture=/Game/Effects/Fort_Effects/Emotes/UltraLight/Texture/T_SoftNoise_BC7}.G) {name=DistortionTexture, nodeColor=<rgba 1 0 0.5156 1>}
  n29 = Constant() {r=0.5}
  n30 = LinearInterpolate(Alpha=DynamicParameter() {paramNames [Speed, Reveal, WPO Width, WPO MoweAway]}.Reveal) {constA=-0.5, constB=1.5}
  n31 = MaterialFunctionCall(Float2=TextureCoordinate()) {materialFunction=&BreakOutFloat2Components}
  n32 = Constant() {r=0.8}
  n33 = Add(
   A=SmoothStep(Min=@n30, Max=Add(A=@n30, B=@n29), Value=OneMinus(None=@n31.G))
   B=LinearInterpolate(B=Multiply(A=@n32) {constB=-1}, Alpha=DynamicParameter() {paramNames [Speed, Reveal, WPO Width, WPO MoweAway]}.Reveal) {constB=-0.2}
  )
  n34 = NamedRerouteUsage()
  NamedRerouteDeclaration(None=SmoothStep(Min=@n33, Max=Add(A=@n33, B=@n32), Value=OneMinus(None=@n34))) {name=Reveal, nodeColor=<rgba 1 0 0.8203 1>}
  NamedRerouteDeclaration(None=@n7.R) {name=Dissolve, nodeColor=<rgba 0.8516 0 1 1>}
  NamedRerouteUsage()
 ]
}
