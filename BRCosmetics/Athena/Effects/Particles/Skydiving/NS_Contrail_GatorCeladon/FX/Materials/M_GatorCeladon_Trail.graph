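# Translucent unlit trail material for the GatorCeladon skydiving contrail, built for Niagara ribbons, sprites and mesh particles.
# Panning wispy fire noise, noise-driven distortion and subtraction masks shape the streak, while a Voronoi-broken gradient helps index a curve atlas for color.
# Opacity uses edge masks, optional quality-switched depth fade and camera-distance fade.
# World position offset combines SplineThicken width, sine waves, move-away motion and clamped bending, modulated by Niagara dynamic parameters.
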Material /BRCosmetics/Athena/Effects/Particles/Skydiving/NS_Contrail_GatorCeladon/FX/Materials/M_GatorCeladon_Trail
graph {
 properties {
  bAutomaticallySetUsageInEditor=false
  bUsedWithNiagaraMeshParticles=true
  bUsedWithNiagaraRibbons=true
  bUsedWithNiagaraSprites=true
  blendMode=BLEND_Translucent
  shadingModel=MSM_Unlit
  usedShadingModels=Unlit
 }
 nodes [
  n1 = NamedRerouteUsage()
  n2 = ScalarParameter() {defaultValue=0.2, group=Distortion, parameterName=DistortionAmount, sortPriority=3}
  n3 = Add(A=NamedRerouteUsage(), B=AppendVector(A=Constant(), B=Multiply(A=@n1, B=@n2) {constB=0.2}))
  n4 = VectorParameter() {
   group=MainTexture
   parameterName=FlameTextureUVControl
   sortPriority=2
   channelNames {a=SpeedY, b=SpeedX, g=ScaleY, r=ScaleX}
   defaultValue=<rgba 0.7 1.3 -0.3 0>
  }
  n5 = ComponentMask(None=@n4) {g=true, r=true}
  n6 = TextureSampleParameter2D(UVs=Panner(
    Coordinate=Multiply(A=@n3, B=@n5)
    Time=DynamicParameter() {paramNames [Speed, WPO Disturbance, WPO Width, WPO MoweAway]}.Speed
    Speed=AppendVector(A=@n4.SpeedX, B=@n4.SpeedX)
   ) {speedX=0.6}) {
   group=MainTexture
   parameterName=MainTexture
   samplerType=SAMPLERTYPE_Grayscale
   sortPriority=1
   texture=/Game/Effects/Fort_Effects/Textures/Fire/T_WispyFireNoise_Soft_M
  }
  n7 = NamedRerouteUsage()
  n8 = LinearInterpolate(
   A=Add(A=ChannelMaskParameter(None=@n6.RGB) {group=MainTexture, parameterName=ChannelMask, sortPriority=5}.R, B=@n7)
   B=@n7
   Alpha=ScalarParameter() {parameterName=OnlyCore?}
  )
  n9 = ScalarParameter() {parameterName=TextureSubtract}
  n10 = Subtract(A=Saturate(None=Subtract(A=@n8, B=@n9) {constB=0.5}), B=NamedRerouteUsage())
  n11 = Saturate(None=@n10)
  n12 = ScalarParameter() {defaultValue=1, group=Color, parameterName=ColorPower}
  n13 = ScalarParameter() {defaultValue=0.4, parameterName=Smooth}
  n14 = TextureCoordinate()
  n15 = VectorParameter() {parameterName=DissolveUVControl, defaultValue=<rgba 2 2 -2 0>}
  n16 = MaterialFunctionCall(X=@n15.R, Y=@n15.R) {materialFunction=&MakeFloat2}
  n17 = TextureSample(UVs=Panner(
    Coordinate=Multiply(A=@n14, B=@n16.Result)
    Speed=MaterialFunctionCall(X=@n15.B, Y=@n15.B) {materialFunction=&MakeFloat2}.Result
   )) {texture=/BRCosmetics/Athena/Effects/Particles/Skydiving/NS_Contrail_PinkSpike/Textures/T_PinkSpike_Voronoi2D_Height}
  n18 = OneMinus(None=MaterialFunctionCall(Float2=TextureCoordinate()) {materialFunction=&BreakOutFloat2Components}.R)
  n19 = Constant() {r=3}
  n20 = Add(A=SmoothStep(Min=@n17.R, Max=Add(A=@n17.R, B=@n13), Value=Power(Base=@n18, Exp=@n19)), B=NamedRerouteUsage())
  n21 = Multiply(A=MaterialFunctionCall(Float2=TextureCoordinate()) {materialFunction=&BreakOutFloat2Components}.G) {constB=0.5}
  n22 = LinearInterpolate(A=Constant() {r=0.3}, Alpha=Power(Base=Sine(None=@n21), Exp=ScalarParameter() {defaultValue=8, parameterName=SideColorPower}))
  n23 = CurveAtlasRowParameter(CurveTime=Multiply(A=Power(Base=@n11, Exp=@n12) {constExponent=3}, B=Multiply(A=@n20, B=@n22) {constA=1})) {
   atlas=/BRCosmetics/Athena/Effects/Particles/Skydiving/NS_Contrail_GatorCeladon/FX/Misc/CA_GatorCeladon
   curve=/BRCosmetics/Athena/Effects/Particles/Skydiving/NS_Contrail_GatorCeladon/FX/Misc/CC_GatorCeladon
   parameterName=Curve_Color
  }
  n24 = ParticleColor()
  MaterialFunctionCall(Emissive=Desaturation(None=Multiply(A=@n23, B=@n24.RGB), Fraction=ScalarParameter() {group=Color, parameterName=Desaturation, sortPriority=3})) {materialFunction=&MF_FX_EmissiveAdjustments}
  n25 = Multiply(A=@n10, B=NamedRerouteUsage())
  n26 = ScalarParameter() {defaultValue=3, group=Color, parameterName=AlphaBoost, sortPriority=4}
  n27 = Power(Base=Saturate(None=Multiply(A=@n25, B=@n26) {constB=3}), Exp=ScalarParameter() {defaultValue=1, group=Color, parameterName=AlphaPower})
  n28 = ScalarParameter() {defaultValue=0.2, parameterName=EdgeMaskMax}
  n29 = ComponentMask(None=TextureCoordinate()) {r=true}
  n30 = ComponentMask(None=TextureCoordinate()) {r=true}
  n31 = Multiply(
   A=Multiply(A=Multiply(A=@n24.A, B=@n27), B=SmoothStep(Max=@n28, Value=@n29))
   B=SmoothStep(Max=ScalarParameter() {parameterName=TrailingMaskMax}, Value=OneMinus(None=@n30))
  )
  n32 = DepthFade(Opacity=@n31, FadeDistance=ScalarParameter() {parameterName=DepthFade})
  Multiply(
   A=LinearInterpolate(A=@n31, B=QualitySwitch(Default=@n32, Low=@n31, High=@n32, Medium=@n31, Epic=@n32), Alpha=ScalarParameter() {parameterName=UseDepthFade})
   B=MaterialFunctionCall!(Distance=ScalarParameter() {defaultValue=200, parameterName=CameraDistanceFade}) {materialFunction=/Game/Effects/Fort_Effects/Materials/_Functions/MF_CameraDistanceFade}.Result
  )
  n33 = Multiply(
   A=Multiply(
    A=ScalarParameter() {defaultValue=1, parameterName=WidthMainMult}
    B=DynamicParameter() {paramNames [Speed, WPO Disturbance, WPO Width, WPO MoweAway]}.WPO Width
   )
   B=MaterialFunctionCall(Input=VertexColor().R, Input Low=Constant(), Input High=Constant() {r=1}, Target Low=Constant() {r=0.8}, Target High=Constant() {r=1.3}) {materialFunction=&RemapValueRange}.Result
  )
  n34 = MaterialFunctionCall(
   WidthBase=Multiply(A=ScalarParameter() {defaultValue=40, parameterName=WidthBase}, B=@n33)
   WidthTip=Multiply(A=ScalarParameter() {defaultValue=100, parameterName=WidthTip}, B=@n33)
   Expand U or V UV Channel=StaticBool()
  ) {materialFunction=&SplineThicken}
  n35 = ScalarParameter() {defaultValue=1, parameterName=UseSplineThicken}
  n36 = ComponentMask(None=TextureCoordinate()) {r=true}
  n37 = ScalarParameter() {defaultValue=0.5, parameterName=WPODistort_Speed1}
  n38 = Constant()
  n39 = Add(A=NamedRerouteUsage(), B=VertexColor().R)
  n40 = Sine(None=Panner(
    Coordinate=@n39
    Speed=MaterialFunctionCall(X=@n37, Y=@n38) {materialFunction=&MakeFloat2}.Result
   ) {speedX=0.7})
  n41 = Panner(
   Coordinate=@n39
   Speed=MaterialFunctionCall(X=ScalarParameter() {defaultValue=0.3, parameterName=WPODistort_Speed2}, Y=@n38) {materialFunction=&MakeFloat2}.Result
  ) {speedX=0.4}
  n42 = Multiply(A=Sine(None=Multiply(A=@n41) {constB=2})) {constB=0.4}
  n43 = Multiply(A=ComponentMask(None=Multiply(A=@n40, B=@n42)) {r=true}) {constB=2}
  n44 = VectorParameter() {defaultValue=<rgba 1 0 0 1>}
  n45 = Multiply(A=Power(Base=@n36), B=Multiply(A=Subtract(A=@n43) {constB=0}, B=Transform(None=@n44.RGB) {transformSourceType=TRANSFORMSOURCE_Instance}))
  n46 = ScalarParameter() {defaultValue=15, parameterName=WPOIntensity}
  n47 = Add(A=Add(A=LinearInterpolate(B=@n34.WorldPositionOffset, Alpha=@n35), B=Multiply(A=@n45, B=@n46)) {constB=0}, B=NamedRerouteUsage()) {constB=0}
  n48 = NamedRerouteUsage()
  NamedRerouteDeclaration(None=Add(A=Add(A=@n47, B=@n48) {constB=0}, B=NamedRerouteUsage())) {name=WPO, nodeColor=<rgba 0 1 0.9297 1>}
  n49 = OneMinus(None=DynamicParameter() {parameterIndex=1, paramNames [Reveal,Param2,Param3,Param4]}.Reveal)
  n50 = TextureCoordinate()
  n51 = Power(Base=ComponentMask(None=@n50) {r=true}, Exp=ScalarParameter() {defaultValue=3, parameterName=TextureSubtractX}) {constExponent=3}
  n52 = Multiply(A=ComponentMask(None=@n50) {g=true}) {constB=0.5}
  n53 = Power(Base=OneMinus(None=Sine(None=@n52)), Exp=ScalarParameter() {defaultValue=3, parameterName=TextureSubtractY})
  n54 = ComponentMask(None=@n50) {r=true}
  n55 = Constant()
  n56 = Constant() {r=0.05}
  n57 = Constant()
  n58 = LinearInterpolate(
   A=Saturate(None=Add(A=@n51, B=@n53))
   B=Add(A=SphereMask(A=@n54, B=@n55, Radius=@n56, Hardness=@n57), B=@n53)
   Alpha=ScalarParameter() {parameterName=UseRoundTextureSubtract?}
  )
  NamedRerouteDeclaration(None=Saturate(None=Add(A=@n49, B=@n58))) {name=TextureSubtract, nodeColor=<rgba 0 0.2734 1 1>}
  n59 = ScalarParameter() {parameterName=Reveal}
  n60 = Add(A=@n59, B=ScalarParameter() {parameterName=Softness})
  n61 = OneMinus(None=MaterialFunctionCall(Float2=TextureCoordinate()) {materialFunction=&BreakOutFloat2Components}.R)
  NamedRerouteDeclaration(None=Saturate(None=SmoothStep(Min=@n59, Max=@n60, Value=@n61))) {name=Reveal, nodeColor=<rgba 0.2188 0 1 1>}
  n62 = DynamicParameter() {desc=Parameters set by Niagara, parameterIndex=3, paramNames [X,Y,Z,WPOIntensity]}
  n63 = Subtract(
   A=MaterialFunctionCall(X=@n62.X, Y=@n62.X, Z=@n62.X) {materialFunction=&MakeFloat3}.Result
   B=ParticlePositionWS()
  )
  n64 = ScalarParameter() {defaultValue=50, group=WPO, parameterName=Clamp_max}
  n65 = Power(Base=ComponentMask(None=TextureCoordinate()) {r=true}, Exp=ScalarParameter() {defaultValue=2, group=WPO, parameterName=WPOBendPower})
  n66 = Constant() {r=1.5}
  n67 = LinearInterpolate(
   B=Clamp(None=Multiply(B=@n63) {constA=1}, Min=Multiply(A=@n64) {constB=-1}, Max=@n64) {maxDefault=1000, minDefault=-1000}
   Alpha=Multiply(A=Multiply(A=@n65, B=@n66), B=@n62.WPOIntensity)
  )
  n68 = Constant3Vector() {constant=<rgba 1 1 0 0>}
  NamedRerouteDeclaration(None=LinearInterpolate(B=Multiply(A=@n67, B=@n68), Alpha=ScalarParameter() {defaultValue=1, parameterName=Use interpolate})) {name=WPO Interpolate, nodeColor=<rgba 1 0 0.0703 1>}
  n69 = TextureCoordinate()
  n70 = Power(Base=OneMinus(None=ComponentMask(None=@n69) {r=true})) {constExponent=1}
  n71 = Multiply(A=ComponentMask(None=@n69) {g=true}) {constB=0.5}
  n72 = Multiply(A=Saturate(None=Multiply(A=@n70) {constB=2}), B=Power(Base=Sine(None=@n71), Exp=ScalarParameter() {defaultValue=3, parameterName=SidePower}))
  n73 = Add(A=TextureCoordinate(), B=Multiply(A=NamedRerouteUsage(), B=ScalarParameter() {defaultValue=0.3}))
  n74 = Power(Base=OneMinus(None=ComponentMask(None=@n73) {r=true})) {constExponent=3}
  NamedRerouteDeclaration(None=Multiply(A=Multiply(A=@n72, B=@n74), B=Constant() {r=1})) {name=Core, nodeColor=<rgba 0 1 0.7656 1>}
  n75 = NamedRerouteUsage()
  n76 = VectorParameter() {
   group=Distortion
   parameterName=DistortionUVControl
   sortPriority=2
   channelNames {a=SpeedY, b=SpeedX, g=ScaleY, r=ScaleX}
   defaultValue=<rgba 0.8 0.8 -0.5 0>
  }
  n77 = ComponentMask(None=@n76) {g=true, r=true}
  n78 = TextureSampleParameter2D(UVs=Panner(
    Coordinate=Multiply(A=@n75, B=@n77) {constB=0.6}
    Time=DynamicParameter() {paramNames [Speed, WPO Disturbance, WPO Width, WPO MoweAway]}.Speed
    Speed=AppendVector(A=@n76.SpeedX, B=@n76.SpeedX)
   ) {speedX=1, speedY=0.25}) {
   group=Distortion
   parameterName=DistortionTexture
   samplerType=SAMPLERTYPE_Grayscale
   sortPriority=1
   texture=&T_WispyFireNoise_Soft_M
  }
  NamedRerouteDeclaration(None=@n78.R) {name=DistortionTexture, nodeColor=<rgba 1 0 0.5156 1>}
  n79 = MaterialFunctionCall(Float2=TextureCoordinate()) {materialFunction=&BreakOutFloat2Components}
  n80 = OneMinus(None=Power(Base=OneMinus(None=@n79.R), Exp=Constant() {r=1.1}))
  n81 = ParticleRandom()
  n82 = Constant()
  n83 = Add(
   A=LinearInterpolate(B=AppendVector(A=@n81, B=@n82), Alpha=ScalarParameter() {parameterName=AddParticleRandom})
   B=AppendVector(A=ScalarParameter() {parameterName=OffsetX}, B=ScalarParameter() {parameterName=OffsetY})
  )
  n84 = VertexColor()
  NamedRerouteDeclaration(None=Add(
    A=MaterialFunctionCall(X=@n80, Y=@n79.G) {materialFunction=&MakeFloat2}.Result
    B=Add(A=@n83, B=@n84.R)
   ) {constB=0}) {name=UVWithOffset, nodeColor=<rgba 0.7578 0 1 1>}
  n85 = VertexColor()
  n86 = MaterialFunctionCall(X=Multiply(A=Add(A=@n85.B) {constB=-0.5}) {constB=2}, Y=Multiply(A=Add(A=@n85.A) {constB=-0.5}) {constB=2}, Z=Constant()) {materialFunction=&MakeFloat3}
  n87 = VertexColor()
  n88 = Multiply(A=@n87.R, B=ScalarParameter() {defaultValue=1, parameterName=TimeOffsetAmount})
  n89 = Add(A=Multiply(A=Add(A=@n87.G, B=@n88) {constB=0}, B=Constant() {r=1}), B=Multiply(A=Time(), B=Constant() {r=1}))
  n90 = Power(Base=ConstantBiasScale(None=Sine(None=@n89)))
  n91 = Multiply(
   A=ScalarParameter() {defaultValue=1, parameterName=MoveAway}
   B=DynamicParameter() {paramNames [Speed, WPO Disturbance, WPO Width, WPO MoweAway]}.WPO MoweAway
  )
  NamedRerouteDeclaration(None=Multiply(A=Transform(None=@n86.Result) {transformSourceType=TRANSFORMSOURCE_Local}, B=Multiply(A=@n90, B=@n91) {constA=1})) {name=MoveAway, nodeColor=<rgba 0 1 0.3672 1>}
  n92 = ComponentMask(None=Add(A=TextureCoordinate() {uTiling=2.2}, B=Constant())) {r=true}
  n93 = Multiply(A=Time()) {constB=-1.3}
  n94 = Multiply(A=Sine(None=Add(A=@n92, B=@n93)) {period=1.2}, B=VectorParameter() {parameterName=WPO Direction, defaultValue=<rgba 0 0 1 1>}.RGB)
  n95 = LinearInterpolate(
   A=Transform(None=@n94) {transformSourceType=TRANSFORMSOURCE_World, transformType=TRANSFORM_Camera}
   B=Transform(None=@n94) {transformSourceType=TRANSFORMSOURCE_Local}
   Alpha=ScalarParameter() {parameterName=Change Transform}
  )
  n96 = DynamicParameter() {paramNames [Speed, WPO Disturbance, WPO Width, WPO MoweAway]}
  n97 = TextureCoordinate()
  n98 = Saturate(None=Multiply(A=ComponentMask(None=@n97) {r=true}) {constB=3})
  NamedRerouteDeclaration(None=Multiply(A=Multiply(A=@n95, B=@n96.WPO Disturbance), B=LinearInterpolate(Alpha=@n98))) {name=WPO_Disturbance, nodeColor=<rgba 0 0.0391 1 1>}
  n99 = OneMinus(None=VertexColor().G)
  n100 = Constant() {r=1.3}
  NamedRerouteDeclaration(None=Multiply(A=ScalarParameter() {defaultValue=1, parameterName=EmissiveBOoost}, B=Power(Base=@n99, Exp=@n100))) {name=EmissiveBoost, nodeColor=<rgba 0.9219 1 0 1>}
  n101 = Constant3Vector() {constant=<rgba 0.5 0.5 0 0>}
  n102 = SphereMask(
   A=TextureCoordinate()
   B=MaterialFunctionCall(X=@n101.R, Y=@n101.R) {materialFunction=&MakeFloat2}.Result
  ) {attenuationRadius=0.5, hardnessPercent=0}
  NamedRerouteDeclaration(None=LinearInterpolate(B=Multiply(A=@n102) {constB=3}, Alpha=ScalarParameter() {parameterName=UseFrontMask?}) {constA=1}) {name=Front, nodeColor=<rgba 0 1 0.9063 1>}
  NamedRerouteUsage()
  OneMinus(None=@n36)
 ]
}
